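# Shadow physics asset for the Flinger enemy, named for use as a simplified capsule proxy of the body for shadowing.
# Single capsules cover pelvis, spine, head, both arms and selected bones along each leg chain, with asymmetric arm sizes and leg bone picks.
# Constraints link the capsules in a tree from the pelvis, with all swing and twist motion left free, so no joint has angular limits.
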PhysicsAsset /Game/Characters/Enemies/Flinger/Mesh/Flinger_ShadowPhysics
bodies {
 Spine_01 [{shapeType=Capsule, length=19.9053, radius=7.2507, center=<15.1158 -2.0473 -0.6645>, rotation=<rot -0.3743 -94.5587 -86.8923>}]
 hand_l [{shapeType=Capsule, length=44.5017, radius=4.9876, center=<22.3104 -0.2368 -0.5652>, rotation=<rot 0 -89.2603 -85.4872>}]
 hand_r [{shapeType=Capsule, length=36.9338, radius=14.0733, center=<-14.5327 2.4298 1.6524>, rotation=<rot 4.6068 91.2989 -73.8247>}]
 head [{shapeType=Capsule, length=27.8522, radius=13.8379, center=<-7.9897 11.0227 -0.1167>, rotation=<rot 0 34.32 -85.6932>}]
 leg_chain_l_01 [{shapeType=Capsule, length=23.3524, radius=10.2533, center=<0.3228 -10.1337 -0.1476>, rotation=<rot 0 2.4153 -84.4139>}]
 leg_chain_l_03 [{shapeType=Capsule, length=19.8283, radius=7.48, center=<0.4361 -15.0754 -0.1591>, rotation=<rot 0 179.3995 -88.2118>}]
 leg_chain_l_06 [{shapeType=Capsule, length=37.131, radius=6.421, center=<-1.2932 -19.5036 0.0296>, rotation=<rot 0 172.6258 -87.1402>}]
 leg_chain_l_09 [{shapeType=Capsule, length=42.4579, radius=3.1367, center=<-0.0085 -19.1551 0.5145>, rotation=<rot 0 -0.2844 -88.1155>}]
 leg_chain_r_02 [{shapeType=Capsule, length=32.2565, radius=7.2148, center=<0.5584 -8.362 0.1786>, rotation=<rot 0 -178.7837 -89.5695>}]
 leg_chain_r_05 [{shapeType=Capsule, length=34.5811, radius=6.867, center=<-1.692 -19.3071 -0.4588>, rotation=<rot 0 172.3635 -86.8194>}]
 leg_chain_r_08 [{shapeType=Capsule, length=45.7481, radius=2.6742, center=<0.5184 -24.3437 -0.9069>, rotation=<rot 0 -178.7603 -87.2048>}]
 lowerarm_l [{shapeType=Capsule, length=73.5229, radius=6.1199, center=<35.6132 0.4884 0.7604>, rotation=<rot 0 87.777 -89.4023>}]
 lowerarm_r [{shapeType=Capsule, length=46.7059, radius=16.9373, center=<-34.6513 0 -4.4504>, rotation=<rot 0 -89.9999 -93.1095>}]
 pelvis [{shapeType=Capsule, length=31.4341, radius=10.1276, center=<1.2978 -0.1584 -1.4365>, rotation=<rot 12.1032 -90 -5.9756>}]
 spine_05 [{shapeType=Capsule, length=22.0709, radius=32.5816, center=<6.805 0 1.3039>, rotation=<rot 0 -90 0>}]
 upperarm_l [{shapeType=Capsule, length=32.9556, radius=6.1761, center=<27.599 -0.2232 0.288>, rotation=<rot 0 -88.9088 -84.4354>}]
 upperarm_r [{shapeType=Capsule, length=21.3375, radius=8.2013, center=<-25.191 -2.0321 0.2977>, rotation=<rot -1.4203 100.4596 -96.3313>}]
}
constraints [
 {
  bone1Name=Spine_01
  bone2Name=pelvis
  swing1LimitDegrees=45
  swing1Motion=Free
  swing2LimitDegrees=45
  swing2Motion=Free
  twistLimitDegrees=45
  twistMotion=Free
 }
 {
  bone1Name=upperarm_l
  bone2Name=spine_05
  swing1LimitDegrees=45
  swing1Motion=Free
  swing2LimitDegrees=45
  swing2Motion=Free
  twistLimitDegrees=45
  twistMotion=Free
 }
 {
  bone1Name=lowerarm_l
  bone2Name=upperarm_l
  swing1LimitDegrees=45
  swing1Motion=Free
  swing2LimitDegrees=45
  swing2Motion=Free
  twistLimitDegrees=45
  twistMotion=Free
 }
 {
  bone1Name=hand_l
  bone2Name=lowerarm_l
  swing1LimitDegrees=45
  swing1Motion=Free
  swing2LimitDegrees=45
  swing2Motion=Free
  twistLimitDegrees=45
  twistMotion=Free
 }
 {
  bone1Name=upperarm_r
  bone2Name=spine_05
  swing1LimitDegrees=45
  swing1Motion=Free
  swing2LimitDegrees=45
  swing2Motion=Free
  twistLimitDegrees=45
  twistMotion=Free
 }
 {
  bone1Name=hand_r
  bone2Name=lowerarm_r
  swing1LimitDegrees=45
  swing1Motion=Free
  swing2LimitDegrees=45
  swing2Motion=Free
  twistLimitDegrees=45
  twistMotion=Free
 }
 {
  bone1Name=leg_chain_l_01
  bone2Name=pelvis
  swing1LimitDegrees=45
  swing1Motion=Free
  swing2LimitDegrees=45
  swing2Motion=Free
  twistLimitDegrees=45
  twistMotion=Free
 }
 {
  bone1Name=leg_chain_l_03
  bone2Name=leg_chain_l_01
  swing1LimitDegrees=45
  swing1Motion=Free
  swing2LimitDegrees=45
  swing2Motion=Free
  twistLimitDegrees=45
  twistMotion=Free
 }
 {
  bone1Name=leg_chain_l_06
  bone2Name=leg_chain_l_03
  swing1LimitDegrees=45
  swing1Motion=Free
  swing2LimitDegrees=45
  swing2Motion=Free
  twistLimitDegrees=45
  twistMotion=Free
 }
 {
  bone1Name=leg_chain_l_09
  bone2Name=leg_chain_l_06
  swing1LimitDegrees=45
  swing1Motion=Free
  swing2LimitDegrees=45
  swing2Motion=Free
  twistLimitDegrees=45
  twistMotion=Free
 }
 {
  bone1Name=leg_chain_r_02
  bone2Name=pelvis
  swing1LimitDegrees=45
  swing1Motion=Free
  swing2LimitDegrees=45
  swing2Motion=Free
  twistLimitDegrees=45
  twistMotion=Free
 }
 {
  bone1Name=leg_chain_r_05
  bone2Name=leg_chain_r_02
  swing1LimitDegrees=45
  swing1Motion=Free
  swing2LimitDegrees=45
  swing2Motion=Free
  twistLimitDegrees=45
  twistMotion=Free
 }
 {
  bone1Name=leg_chain_r_08
  bone2Name=leg_chain_r_05
  swing1LimitDegrees=45
  swing1Motion=Free
  swing2LimitDegrees=45
  swing2Motion=Free
  twistLimitDegrees=45
  twistMotion=Free
 }
 {
  bone1Name=lowerarm_r
  bone2Name=upperarm_r
  swing1LimitDegrees=45
  swing1Motion=Free
  swing2LimitDegrees=45
  swing2Motion=Free
  twistLimitDegrees=45
  twistMotion=Free
 }
 {
  bone1Name=spine_05
  bone2Name=Spine_01
  swing1LimitDegrees=45
  swing1Motion=Free
  swing2LimitDegrees=45
  swing2Motion=Free
  twistLimitDegrees=45
  twistMotion=Free
 }
 {
  bone1Name=head
  bone2Name=spine_05
  swing1LimitDegrees=45
  swing1Motion=Free
  swing2LimitDegrees=45
  swing2Motion=Free
  twistLimitDegrees=45
  twistMotion=Free
 }
]
